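# Level Sequence for day 100 of the PistolCanary countdown, the "anm" sequence that hides the MaskMaker cinematic NPC.
# It possesses that single actor and drives a visibility track whose channel stays false.
# It has no animation, transform or root-level tracks, and runs at a 60 fps display rate.
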LevelSequence /PistolCanary_Countdown/Sequence/day/day100/PistolCanary_Countdown_day100_anm
bindings [PistolCanary_CinematicNPC_MaskMaker {
  id=#1
  possessed_class=PistolCanary_CinematicNPC_MaskMaker_C
  tracks [{
    type=MovieSceneVisibilityTrack
    sections [{type=MovieSceneVisibilitySection, channels [None {type=MovieSceneScriptingBoolChannel, default=false, const [false,1]}]}]
   }]
 }]
display_rate {denominator=1, numerator=60}
tick_resolution {denominator=1, numerator=24000}
tracks []
