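# Level Sequence for day 100 of PistolCanary_Countdown that controls the visibility of the game countdown actor.
# It possesses a BP_Generic_Game_Countdown actor bound as PistolCanary_BP_Countdown.
# A single visibility track holds that actor at a constant true value, with no other tracks in the sequence.
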LevelSequence /PistolCanary_Countdown/Sequence/day/day100/PistolCanary_Countdown_day100_gam
bindings [PistolCanary_BP_Countdown {
  id=#1
  possessed_class=BP_Generic_Game_Countdown_C
  tracks [{
    type=MovieSceneVisibilityTrack
    sections [{type=MovieSceneVisibilitySection, channels [None {type=MovieSceneScriptingBoolChannel, default=true, const [true,1]}]}]
   }]
 }]
display_rate {denominator=1, numerator=60}
tick_resolution {denominator=1, numerator=24000}
tracks []
